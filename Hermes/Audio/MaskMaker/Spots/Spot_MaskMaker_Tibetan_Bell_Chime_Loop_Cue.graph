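# Looping ambient spot cue for a Tibetan bell ring at the MaskMaker location, in the Athena ambient props sound class.
# A Random node starts the shared looping wave either immediately or after a half-second delay, staggering instances, with an Enveloper fading volume in.
# Inline overrides set focus-aware attenuation and cap concurrency at two voices.
# It sends to the exterior ambient bus and to reverb, which rises with distance.
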SoundCue /Hermes/Audio/MaskMaker/Spots/Spot_MaskMaker_Tibetan_Bell_Chime_Loop_Cue
root=Random_0
nodes [
 {id=Delay_0, children [Enveloper_2], properties {delayMax=0.5, delayMin=0.5}}
 {
  id=Enveloper_2
  children [WavePlayer_0]
  properties {pitchCurve {editorCurveData {keys [{time=0}]}}, volumeCurve {editorCurveData {keys [<at 0 0>,<at 0.5 1>]}}}
 }
 {id=Random_0, children [Delay_0,Enveloper_2], properties {weights [1,1]}}
 {id=WavePlayer_0, properties {bLooping=true, soundWaveAssetPtr=&/Spot_MaskMaker_Tibetan_Bell_Ring_Loop_01}}
]
properties {
 bOverrideAttenuation=true
 bOverrideConcurrency=true
 maxDistance=1150
 soundClassObject=&Fort_Ambient_Props_Athena
 volumeMultiplier=1
 attenuationOverrides {
  bEnableListenerFocus=true
  falloffDistance=1000
  focusDistanceScale=0.8
  focusVolumeAttenuation=1.2
  nonFocusVolumeAttenuation=0.6
  stereoSpread=10
  attenuationShapeExtents {x=150}
 }
 busSends [{audioBus=null, minSendLevel=0, sendLevel=0.6, soundSourceBus=&AmbientAudio_Exterior_SourceBus}]
 concurrencyOverrides {maxCount=2, voiceStealReleaseTime=2}
 soundSubmixSends [{
   maxSendDistance=8000
   minSendLevel=0.25
   soundSubmix=&Fortnite_Reverb_Default
   customSendLevelCurve {editorCurveData {keys [<at 0 0.25>,<at 8000 1>]}}
  }]
}
